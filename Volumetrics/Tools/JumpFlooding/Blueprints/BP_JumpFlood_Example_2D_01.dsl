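; Example Blueprint that demonstrates the 2D jump flood component, building a distance field and gradient from a mask texture and previewing it on a mesh.
; Construction script assigns a debug distance-field MID to the preview mesh and sets its grid, distance, gradient and divisor parameters.
; Numbered custom events clear both render targets and reset indices, run the jump flood, or compute the gradient, feeding results to the MID's RT parameter.
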(event Custom|2%29%20Jump%20Flood
  ;[1 1 1] How to use the component when specifying a mask texture
  (bind _result_rt (Class|JumpFloodComponent2D|JumpFlood (Variables|Default|GetJumpFloodComponent2D) (Variables|Default|GetSeed%20Mode) (Variables|Default|GetMask%20Texture) 0 :!class /Volumetrics/Tools/JumpFlooding/Blueprints/JumpFloodComponent_2D.JumpFloodComponent_2D_C))
  (Rendering|Material|SetTextureParameterValue (Variables|MIDs|GetDebug%20RT%20MID) "RT" _result_rt))

(event Custom|1%29%20Clear%20RTs
  (bind _jumpfloodcomponent2d (Variables|Default|GetJumpFloodComponent2D))
  (bind _jumpfloodcomponent2d_2 (Variables|Default|GetJumpFloodComponent2D))
  ;[0.4896 0.004675 0] Debug Reset Button
  (Rendering|ClearRenderTarget2D (Variables|/Volumetrics/Tools/JumpFlooding/Blueprints/JumpFloodComponent_2D.JumpFloodComponent_2D_C|GetRT%20A _jumpfloodcomponent2d))
  (Rendering|ClearRenderTarget2D (Variables|/Volumetrics/Tools/JumpFlooding/Blueprints/JumpFloodComponent_2D.JumpFloodComponent_2D_C|GetRT%20B _jumpfloodcomponent2d))
  (Variables|/Volumetrics/Tools/JumpFlooding/Blueprints/JumpFloodComponent_2D.JumpFloodComponent_2D_C|SetCurrentIndex 0 _jumpfloodcomponent2d_2)
  (Variables|/Volumetrics/Tools/JumpFlooding/Blueprints/JumpFloodComponent_2D.JumpFloodComponent_2D_C|SetStartIndex 0 _jumpfloodcomponent2d_2))

(event Custom|3%29%20Gradient
  ;[1 1 1] Gradient
  (bind _result_rt (Class|JumpFloodComponent2D|ComputeGradient (Variables|Default|GetJumpFloodComponent2D)))
  (Rendering|Material|SetTextureParameterValue (Variables|MIDs|GetDebug%20RT%20MID) "RT" _result_rt))

(fn ConstructionScript
  (bind _debug_rt_mid (Variables|MIDs|GetDebug%20RT%20MID))
  (bind _debug_rt_mid_2 (Variables|MIDs|GetDebug%20RT%20MID))
  (Class|VolumetricsBPLIbrary|CreateAndAssignMID _debug_rt_mid "/Volumetrics/Tools/JumpFlooding/Materials/Debug/M_Debug_DistanceField_Texels.M_Debug_DistanceField_Texels")
  (Rendering|Material|SetMaterial (Variables|Default|GetPreview%20Mesh) 0 _debug_rt_mid)
  (Rendering|Material|SetScalarParameterValue _debug_rt_mid_2 "ShowGrid" (Math|Conversions|ToFloat(Boolean) (Variables|Default|GetShow%20Grid)) :!class /Script/Engine.MaterialInstanceDynamic)
  (Rendering|Material|SetScalarParameterValue _debug_rt_mid_2 "ShowDistance" (Math|Conversions|ToFloat(Boolean) (Variables|Default|GetShow%20Distance)) :!class /Script/Engine.MaterialInstanceDynamic)
  (Rendering|Material|SetScalarParameterValue _debug_rt_mid_2 "ShowGradient" (Math|Conversions|ToFloat(Boolean) (Variables|Default|GetShow%20Gradient)) :!class /Script/Engine.MaterialInstanceDynamic)
  (Rendering|Material|SetScalarParameterValue _debug_rt_mid_2 "DistanceDivisor" (Variables|Default|GetDistance%20Divisor) :!class /Script/Engine.MaterialInstanceDynamic))
