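; Blueprint component that runs 2D jump flooding on a seed texture using material draws into two ping-ponged render targets.
; JumpFlood allocates RGBA16f targets, detects edges from a selected channel unless the seed mode skips it, optionally smooths, then loops jump steps.
; A start index tracks buffer parity, letting the seed be one of the targets.
; Separate functions compute distance and gradient or assign external targets, warning on size mismatch.
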(fn AllocateRenderTargets
  (params (Seed%20Texture :type Texture%20Object%20Reference%20%28by%20ref%29))
  (bind _size (Class|VolumetricsBPLIbrary|GetTextureSize Seed%20Texture))
  (Variables|Default|Set2D%20Size _size)
  (Class|VolumetricsBPLIbrary|CreateAndAssignRT (Variables|Default|GetRT%20A) (Variables|Default|Get2D%20Size) "RTF_RGBA16f")
  (Class|VolumetricsBPLIbrary|CreateAndAssignRT (Variables|Default|GetRT%20B) (Variables|Default|Get2D%20Size) "RTF_RGBA16f")
  (Default|ComputeRequiredIterations))

(fn AssignRenderTargets
  (params (RT%20A :type Texture%20Render%20Target%202D%20Object%20Reference%20%28by%20ref%29) (RT%20B :type Texture%20Render%20Target%202D%20Object%20Reference%20%28by%20ref%29))
  (Variables|Default|SetRT%20A RT%20A)
  (Variables|Default|SetRT%20B RT%20B)
  (bind _size (Class|VolumetricsBPLIbrary|GetTextureSize (Variables|Default|GetRT%20A)))
  (Variables|Default|Set2D%20Size _size)
  (Default|ComputeRequiredIterations
    (:Out
      (bind _match (Class|VolumetricsBPLIbrary|RTSizeCompare (Variables|Default|GetRT%20A) (Variables|Default|GetRT%20B)))
      (if (not _match)
        ;[1 1 1] Warn on Resolution Mismatch
        (Development|PrintString (Utilities|String|Append "Warning: Supplied Render Targets do not have equal resolution. Artifacts may result."))
        (Development|PrintString (Utilities|String|Append "Supplied Render Target A Resolution: " (Utilities|IntPointToString (Utilities|Struct|MakeIntPoint (Variables|/Script/Engine.TextureRenderTarget2D|GetSizeX (Variables|Default|GetRT%20A)) (Variables|/Script/Engine.TextureRenderTarget2D|GetSizeY (Variables|Default|GetRT%20A)))) "Supplied Render Target B Resolution: " (Utilities|IntPointToString (Utilities|Struct|MakeIntPoint (Variables|/Script/Engine.TextureRenderTarget2D|GetSizeX (Variables|Default|GetRT%20B)) (Variables|/Script/Engine.TextureRenderTarget2D|GetSizeY (Variables|Default|GetRT%20B))))))))))

(fn ComputeDistance
  (bind (_read_target _write_target) (Utilities|PingPong))
  (bind _compute_distance_mid (Variables|Default|GetCompute%20Distance%20MID))
  (Class|VolumetricsBPLIbrary|CreateAndAssignMID (Variables|Default|GetCompute%20Distance%20MID) "/Volumetrics/Tools/JumpFlooding/Materials/Generation/M_JumpFlood_Distance_2D.M_JumpFlood_Distance_2D")
  (Rendering|Material|SetTextureParameterValue _compute_distance_mid "RT" (Utilities|PingPong))
  (Rendering|Material|SetVectorParameterValue _compute_distance_mid "TextureSize" (Utilities|IntPointToLinearColor (Variables|Default|Get2D%20Size)) :!class /Script/Engine.MaterialInstanceDynamic)
  (Rendering|ClearRenderTarget2D _write_target)
  (Rendering|DrawMaterialtoRenderTarget _write_target (Variables|Default|GetCompute%20Distance%20MID))
  (Math|Integer|IncrementInt (Variables|Default|GetCurrentIndex)
    (:%20%20
      (return (Utilities|PingPong)))))

(fn ComputeGradient
  (bind (_read_target _write_target) (Utilities|PingPong))
  (bind _compute_gradient_mid (Variables|Default|GetCompute%20Gradient%20MID))
  (Class|VolumetricsBPLIbrary|CreateAndAssignMID (Variables|Default|GetCompute%20Gradient%20MID) (Variables|Default|GetCompute%20Gradient%20Material))
  (Rendering|Material|SetTextureParameterValue _compute_gradient_mid "RT" (Utilities|PingPong))
  (Rendering|Material|SetVectorParameterValue _compute_gradient_mid "TextureSize" (Utilities|IntPointToLinearColor (Variables|Default|Get2D%20Size)) :!class /Script/Engine.MaterialInstanceDynamic)
  (Rendering|ClearRenderTarget2D _write_target)
  (Rendering|DrawMaterialtoRenderTarget _write_target (Variables|Default|GetCompute%20Gradient%20MID))
  (Math|Integer|IncrementInt (Variables|Default|GetCurrentIndex)
    (:%20%20
      (return (Utilities|PingPong)))))

(fn DetectEdges
  (params (Seed%20Texture :type Texture%20Object%20Reference%20%28by%20ref%29) (Mode :type JumpFloodSeedMode%20Enum :default "NewEnumerator0"))
  (bind (_read_target _write_target) (Utilities|PingPong))
  (bind _find_edges_mid (Variables|Default|GetFind%20Edges%20MID))
  (bind _output_get (Variables|Default|SetSeed%20Texture Seed%20Texture))
  (Utilities|CheckStartIndex _output_get
    (:Out
      (Variables|Default|SetSeedMode Mode)
      ;[1 1 1] MID Setup
      (Class|VolumetricsBPLIbrary|CreateAndAssignMID (Variables|Default|GetFind%20Edges%20MID) (Variables|Default|GetFind%20Edges%20Material))
      ;[1 0.7372 0.2304] MID Parameters
      (Rendering|Material|SetVectorParameterValue _find_edges_mid "TextureResolution" (Utilities|IntPointToLinearColor (Variables|GetOutPoint)) :!class /Script/Engine.MaterialInstanceDynamic)
      (Rendering|Material|SetVectorParameterValue _find_edges_mid "MaskChannel" (Utilities|Select :Index (Variables|Default|GetSeedMode) :NewEnumerator0 "(R=1.000000,G=0.000000,B=0.000000,A=0.000000)" :NewEnumerator2 "(R=0.000000,G=1.000000,B=0.000000,A=0.000000)" :NewEnumerator3 "(R=0.000000,G=0.000000,B=1.000000,A=0.000000)" :NewEnumerator4 "(R=0.000000,G=0.000000,B=0.000000,A=1.000000)" :NewEnumerator1 "(R=0.000000,G=0.000000,B=0.000000,A=0.000000)") :!class /Script/Engine.MaterialInstanceDynamic)
      (Rendering|Material|SetTextureParameterValue _find_edges_mid "Seed" (Variables|Default|GetSeed%20Texture))
      ;[0.1527 0.3856 1] Draw
      (Rendering|ClearRenderTarget2D _write_target)
      (Rendering|DrawMaterialtoRenderTarget _write_target (Variables|Default|GetFind%20Edges%20MID))
      (Math|Integer|IncrementInt (Variables|Default|GetStartIndex)
        (:%20%20
          (return (Utilities|PingPong)))))) ; Allows seed to be one of the RTs
)

(fn EdgeSmoothIteration
  (bind (_read_target _write_target) (Utilities|PingPong))
  (bind _edge_smooth_mid (Variables|Default|GetEdge%20Smooth%20MID))
  ;[1 1 1] MID Setup
  (Class|VolumetricsBPLIbrary|CreateAndAssignMID (Variables|Default|GetEdge%20Smooth%20MID) (Variables|Default|GetEdge%20Smooth%20Material))
  (Rendering|Material|SetTextureParameterValue _edge_smooth_mid "RT" (Utilities|PingPong))
  (Rendering|Material|SetVectorParameterValue _edge_smooth_mid "TextureResolution" (Utilities|IntPointToLinearColor (Variables|Default|Get2D%20Size)) :!class /Script/Engine.MaterialInstanceDynamic)
  (Rendering|ClearRenderTarget2D _write_target)
  (Rendering|DrawMaterialtoRenderTarget _write_target (Variables|Default|GetEdge%20Smooth%20MID))
  (Math|Integer|IncrementInt (Variables|Default|GetCurrentIndex)
    (:%20%20
      (return (Utilities|PingPong)))))

(fn JumpFlood
  (params (Seed%20Mode :type JumpFloodSeedMode%20Enum :default "NewEnumerator0") (Seed%20Texture :type Texture%20Object%20Reference) (Smoothing%20Passes :type Integer))
  (Variables|Default|SetSeedMode Seed%20Mode)
  (Variables|Default|SetSeed%20Texture Seed%20Texture)
  (Variables|Default|SetSmoothPasses Smoothing%20Passes)
  (CallFunction|AllocateRenderTargets (Variables|Default|GetSeed%20Texture))
  (Variables|Default|SetStartIndex 0)
  (Variables|Default|SetCurrentIndex 0)
  (if (not (Utilities|Enum|Equal(Enum) (Variables|Default|GetSeedMode) "NewEnumerator1"))
    (CallFunction|DetectEdges (Variables|Default|GetSeed%20Texture) (Variables|Default|GetSeedMode))
    (else
      (Utilities|CheckStartIndex (Variables|Default|GetSeed%20Texture)
        (:Out
          _))))
  (if (> (Variables|Default|GetSmoothPasses) 0)
    ;[1 1 1] Edge Smoothing
    (while (< (Variables|Default|GetCurrentIndex) (Variables|Default|GetSmoothPasses))
      (CallFunction|EdgeSmoothIteration))
    (Variables|Default|SetStartIndex (+ (Variables|Default|GetStartIndex) (Math|Integer|%25%28Integer%29 (Variables|Default|GetSmoothPasses) 2))))
  ;[0.2142 0.4953 1] Jump Flood Iterations
  (Variables|Default|SetCurrentIndex 0)
  (while (< (Variables|Default|GetCurrentIndex) (Variables|Default|GetLastIndex))
    (CallFunction|SingleIteration))
  (return (Utilities|PingPong)))

(fn SingleIteration
  (bind (_read_target _write_target) (Utilities|PingPong))
  (bind _jump_step_mid (Variables|Default|GetJump%20Step%20MID))
  (Class|VolumetricsBPLIbrary|CreateAndAssignMID (Variables|Default|GetJump%20Step%20MID) (Variables|Default|GetJump%20Step%20Material))
  (Rendering|Material|SetTextureParameterValue _jump_step_mid "RT" (Utilities|PingPong))
  (Rendering|Material|SetScalarParameterValue _jump_step_mid "Index" (Math|Conversions|ToFloat(Integer) (Variables|Default|GetCurrentIndex)) :!class /Script/Engine.MaterialInstanceDynamic)
  (Rendering|Material|SetVectorParameterValue _jump_step_mid "TextureResolution" (Utilities|IntPointToLinearColor (Variables|Default|Get2D%20Size)) :!class /Script/Engine.MaterialInstanceDynamic)
  (Rendering|ClearRenderTarget2D _write_target)
  (Rendering|DrawMaterialtoRenderTarget _write_target (Variables|Default|GetJump%20Step%20MID))
  (Math|Integer|IncrementInt (Variables|Default|GetCurrentIndex)
    (:%20%20
      (return (Utilities|PingPong)))))
