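; Blueprint component that runs a 3D jump flood on the GPU over pseudo volume textures laid out in 2D render targets.
; JumpFlood allocates two RGBA16f targets, seeds them through an optional channel-masked edge-detection pass, then repeats a jump-step material until the last index.
; Each pass ping-pongs between the targets, with start-index handling that lets a target serve as seed.
; Separate functions assign external targets, warning on size mismatch, and draw a gradient pass.
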(fn AllocateRenderTargets
  (params (3D%20Size :type EVolumeTextureDimensions%20Enum%20%28by%20ref%29 :default "NewEnumerator0"))
  (bind (_rt_size _xy_frames _volume_size) (Class|VolumetricsBPLIbrary|PseudoVolumeDimensions 3D%20Size))
  (Variables|Default|Set2D%20Size _rt_size)
  (Variables|Default|SetVolume%20Size _volume_size)
  (Variables|Default|Set3D%20Size 3D%20Size)
  (Class|VolumetricsBPLIbrary|CreateAndAssignRT (Variables|Default|GetRT%20A) (Variables|Default|Get2D%20Size) "RTF_RGBA16f")
  (Class|VolumetricsBPLIbrary|CreateAndAssignRT (Variables|Default|GetRT%20B) (Variables|Default|Get2D%20Size) "RTF_RGBA16f")
  (Default|ComputeRequiredIterations))

(fn AssignRenderTargets
  (params (RT%20A :type Texture%20Render%20Target%202D%20Object%20Reference%20%28by%20ref%29) (RT%20B :type Texture%20Render%20Target%202D%20Object%20Reference%20%28by%20ref%29) (3D%20Resolution :type EVolumeTextureDimensions%20Enum :default "NewEnumerator3"))
  (bind (_rt_size _xy_frames _volume_size) (Class|VolumetricsBPLIbrary|PseudoVolumeDimensions 3D%20Resolution))
  (Variables|Default|SetRT%20A RT%20A)
  (Variables|Default|SetRT%20B RT%20B)
  (Variables|Default|Set2D%20Size _rt_size)
  (Variables|Default|SetVolume%20Size _volume_size)
  (Variables|Default|Set3D%20Size 3D%20Resolution)
  (bind _size (Class|VolumetricsBPLIbrary|GetTextureSize (Variables|Default|GetRT%20A)))
  (if (not (Utilities|IntPointCompare _size (Variables|Default|Get2D%20Size)))
    (Utilities|FlowControl|Sequence
      (:then_0
        ;[1 1 1] Warn on Resolution Mismatch
        (Development|PrintString (Utilities|String|Append "Warning: Supplied Render Target A does not match resolution of specified Pseudo Volume Texture. May result in artifacts."))
        (Development|PrintString (Utilities|String|Append "Supplied Render Target A Resolution: " (Utilities|IntPointToString (Utilities|Struct|MakeIntPoint (Variables|/Script/Engine.TextureRenderTarget2D|GetSizeX (Variables|Default|GetRT%20A)) (Variables|/Script/Engine.TextureRenderTarget2D|GetSizeY (Variables|Default|GetRT%20A)))) "Expected Resolution: " (Utilities|IntPointToString (Variables|Default|Get2D%20Size)))))
      (:then_1
        _)))
  (bind _size_2 (Class|VolumetricsBPLIbrary|GetTextureSize (Variables|Default|GetRT%20B)))
  (if (not (Utilities|IntPointCompare _size_2 (Variables|Default|Get2D%20Size)))
    (Utilities|FlowControl|Sequence
      (:then_0
        ;[1 1 1] Warn on Resolution Mismatch
        (Development|PrintString (Utilities|String|Append "Warning: Supplied Render Target B does not match resolution of specified Pseudo Volume Texture. May result in artifacts."))
        (Development|PrintString (Utilities|String|Append "Supplied Render Target B Resolution: " (Utilities|IntPointToString (Utilities|Struct|MakeIntPoint (Variables|/Script/Engine.TextureRenderTarget2D|GetSizeX (Variables|Default|GetRT%20B)) (Variables|/Script/Engine.TextureRenderTarget2D|GetSizeY (Variables|Default|GetRT%20B)))) "Expected Resolution: " (Utilities|IntPointToString (Variables|Default|Get2D%20Size)))))
      (:then_1
        _)))
  (Default|ComputeRequiredIterations))

(fn ComputeGradient
  (bind _3d_size (Variables|Default|Get3D%20Size))
  (bind (_rt_size _xy_frames) (Class|VolumetricsBPLIbrary|PseudoVolumeDimensions _3d_size))
  (bind (_read_target _write_target) (Utilities|PingPong))
  (bind _compute_gradient_mid (Variables|Default|GetCompute%20Gradient%20MID))
  (Class|VolumetricsBPLIbrary|CreateAndAssignMID (Variables|Default|GetCompute%20Gradient%20MID) (Variables|Default|GetCompute%20Gradient%20Material))
  (Rendering|Material|SetTextureParameterValue _compute_gradient_mid "RT" (Utilities|PingPong))
  (Rendering|Material|SetVectorParameterValue _compute_gradient_mid "TextureResolution" (Utilities|IntPointToLinearColor _rt_size) :!class /Script/Engine.MaterialInstanceDynamic)
  (Rendering|Material|SetVectorParameterValue _compute_gradient_mid "XYFrames" (Utilities|IntPointToLinearColor _xy_frames) :!class /Script/Engine.MaterialInstanceDynamic)
  (Rendering|ClearRenderTarget2D _write_target)
  (Rendering|DrawMaterialtoRenderTarget _write_target (Variables|Default|GetCompute%20Gradient%20MID))
  (Math|Integer|IncrementInt (Variables|Default|GetCurrentIndex)
    (:%20%20
      (return (Utilities|PingPong)))))

(fn DetectEdges
  (params (Seed%20Texture :type Texture%20Object%20Reference%20%28by%20ref%29) (Mode :type JumpFloodSeedMode%20Enum :default "NewEnumerator0") (3D%20Size :type EVolumeTextureDimensions%20Enum :default "NewEnumerator0"))
  (bind _3d_size (Variables|Default|Get3D%20Size))
  (bind (_rt_size _xy_frames _volume_size) (Class|VolumetricsBPLIbrary|PseudoVolumeDimensions _3d_size))
  (bind (_read_target _write_target) (Utilities|PingPong))
  (bind _find_edges_mid (Variables|Default|GetFind%20Edges%20MID))
  (bind _output_get (Variables|Default|SetSeed%20Texture Seed%20Texture))
  (Utilities|CheckStartIndex _output_get
    (:Out
      (Variables|Default|SetSeedMode Mode)
      (Variables|Default|SetCurrentIndex 0)
      (Variables|Default|Set3D%20Size 3D%20Size)
      ;[1 1 1] MID Setup
      (Class|VolumetricsBPLIbrary|CreateAndAssignMID (Variables|Default|GetFind%20Edges%20MID) (Variables|Default|GetFind%20Edges%20Material))
      ;[1 0.7372 0.2304] MID Parameters
      (Rendering|Material|SetVectorParameterValue _find_edges_mid "TextureResolution" (Utilities|IntPointToLinearColor _rt_size) :!class /Script/Engine.MaterialInstanceDynamic)
      (Rendering|Material|SetVectorParameterValue _find_edges_mid "XYFrames" (Utilities|IntPointToLinearColor _xy_frames) :!class /Script/Engine.MaterialInstanceDynamic)
      (Rendering|Material|SetVectorParameterValue _find_edges_mid "VolumeResolution" (Utilities|IntToLinearColor _volume_size) :!class /Script/Engine.MaterialInstanceDynamic)
      (Rendering|Material|SetVectorParameterValue _find_edges_mid "MaskChannel" (Utilities|Select :Index (Variables|Default|GetSeedMode) :NewEnumerator0 "(R=1.000000,G=0.000000,B=0.000000,A=0.000000)" :NewEnumerator2 "(R=0.000000,G=1.000000,B=0.000000,A=0.000000)" :NewEnumerator3 "(R=0.000000,G=0.000000,B=1.000000,A=0.000000)" :NewEnumerator4 "(R=0.000000,G=0.000000,B=0.000000,A=1.000000)" :NewEnumerator1 "(R=0.000000,G=0.000000,B=0.000000,A=0.000000)") :!class /Script/Engine.MaterialInstanceDynamic)
      (Rendering|Material|SetTextureParameterValue _find_edges_mid "Seed" (Variables|Default|GetSeed%20Texture))
      ;[0.1527 0.3856 1] Draw
      (Rendering|ClearRenderTarget2D _write_target)
      (Rendering|DrawMaterialtoRenderTarget _write_target (Variables|Default|GetFind%20Edges%20MID))
      (Math|Integer|IncrementInt (Variables|Default|GetStartIndex)
        (:%20%20
          (return (Utilities|PingPong)))))) ; Allows seed to be one of the RTs
)

(fn JumpFlood
  (params (Seed%20Mode :type JumpFloodSeedMode%20Enum :default "NewEnumerator0") (Seed%20Texture :type Texture%20Object%20Reference) (3D%20Size :type EVolumeTextureDimensions%20Enum%20%28by%20ref%29 :default "NewEnumerator0"))
  (Variables|Default|SetSeedMode Seed%20Mode)
  (Variables|Default|SetSeed%20Texture Seed%20Texture)
  (bind _output_get (Variables|Default|Set3D%20Size 3D%20Size))
  (CallFunction|AllocateRenderTargets _output_get)
  (Variables|Default|SetStartIndex 0)
  (if (not (Utilities|Enum|Equal(Enum) (Variables|Default|GetSeedMode) "NewEnumerator1"))
    (CallFunction|DetectEdges (Variables|Default|GetSeed%20Texture) (Variables|Default|GetSeedMode) (Variables|Default|Get3D%20Size))
    (else
      (Utilities|CheckStartIndex (Variables|Default|GetSeed%20Texture)
        (:Out
          (Variables|Default|SetCurrentIndex 0)))))
  (while (< (Variables|Default|GetCurrentIndex) (Variables|Default|GetLastIndex))
    (CallFunction|SingleIteration))
  (return (Utilities|PingPong)))

(fn SingleIteration
  (bind _3d_size (Variables|Default|Get3D%20Size))
  (bind (_rt_size _xy_frames _volume_size) (Class|VolumetricsBPLIbrary|PseudoVolumeDimensions _3d_size))
  (bind (_read_target _write_target) (Utilities|PingPong))
  (bind _jump_step_mid (Variables|Default|GetJump%20Step%20MID))
  (bind _jump_step_mid_2 (Variables|Default|GetJump%20Step%20MID))
  (Class|VolumetricsBPLIbrary|CreateAndAssignMID (Variables|Default|GetJump%20Step%20MID) (Variables|Default|GetJump%20Step%20Material))
  ;[1 1 1] MID Config
  (Rendering|Material|SetTextureParameterValue _jump_step_mid_2 "RT" (Utilities|PingPong))
  (Rendering|Material|SetScalarParameterValue _jump_step_mid_2 "Index" (Math|Conversions|ToFloat(Integer) (Variables|Default|GetCurrentIndex)) :!class /Script/Engine.MaterialInstanceDynamic)
  (Rendering|Material|SetVectorParameterValue _jump_step_mid "TextureResolution" (Utilities|IntPointToLinearColor _rt_size) :!class /Script/Engine.MaterialInstanceDynamic)
  (Rendering|Material|SetVectorParameterValue _jump_step_mid "XYFrames" (Utilities|IntPointToLinearColor _xy_frames) :!class /Script/Engine.MaterialInstanceDynamic)
  (Rendering|Material|SetVectorParameterValue _jump_step_mid "VolumeResolution" (Utilities|IntToLinearColor _volume_size) :!class /Script/Engine.MaterialInstanceDynamic)
  (Rendering|ClearRenderTarget2D _write_target)
  (Rendering|DrawMaterialtoRenderTarget _write_target (Variables|Default|GetJump%20Step%20MID))
  (Math|Integer|IncrementInt (Variables|Default|GetCurrentIndex)
    (:%20%20
      (return (Utilities|PingPong)))))
