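# Unlit debug material that graphs the distance field stored in the jump flood render target RT_JumpFlood_01.
# Along the U=0.25 column, it snaps UVs to texel centers and measures the distance to each texel's stored coordinate.
# A mask texture sets the sign, then the distance is scaled and biased.
# PlotFunctionOnGraph draws that curve, with a green copy quantized in 256 steps to compare 8-bit precision.
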Material /Volumetrics/Tools/JumpFlooding/Materials/M_JumpFlood_Display
graph {
 properties {pixelDepthOffsetMode=PDOM_Legacy, shadingModel=MSM_Unlit, usedShadingModels=Unlit}
 nodes [
  n1 = ComponentMask(None=VectorParameter() {parameterName=TextureSize, defaultValue=<rgba 64 64 0 0>}.RGB) {g=true, r=true}
  n2 = Multiply(A=AppendVector(A=Constant() {r=0.25}, B=@n8.x), B=@n1)
  n3 = Divide(A=Divide(B=@n1))
  n4 = Add(A=Divide(A=Floor(None=@n2), B=@n1), B=@n3)
  n5 = ComponentMask(None=TextureSample!(UVs=@n4, Tex, Level, Apply View MipBias) {constMipValue=0, mipValueMode=TMVM_MipLevel, texture=/Volumetrics/Tools/JumpFlooding/Textures/RT_JumpFlood_01}.RGB) {g=true, r=true}
  n6 = Multiply(
   A=MaterialFunctionCall(Vector 2=Subtract(A=@n4, B=@n5)) {materialFunction=/Engine/Functions/Engine_MaterialFunctions02/Utility/VectorLength}.V2 Length
   B=LinearInterpolate(Alpha=TextureSample!(UVs=@n4, Tex, Apply View MipBias) {texture=/Engine/EditorLandscapeResources/DefaultAlphaTexture}.R) {constA=-1}
  )
  n7 = ConstantBiasScale(None=Multiply(A=@n6, B=ScalarParameter() {defaultValue=2.2, parameterName=m}))
  n8 = MaterialFunctionCall!(
   X range Y range=Constant4Vector() {constant=<rgba 0 1 0 1>}.RGBA
   Gradient Height
   ShowGrid?
   Thickness=ScalarParameter() {defaultValue=0.2}
   Color
   "f(x)"=@n7
  ) {materialFunction=/Engine/Functions/Engine_MaterialFunctions01/Debug/PlotFunctionOnGraph}
  n9 = Floor(None=Multiply(A=@n7) {constB=256})
  Add(
   A=@n8."f(x) = y"
   B=MaterialFunctionCall(
    X range Y range=Constant4Vector() {constant=<rgba 0 1 0 1>}.RGBA
    Thickness=ScalarParameter() {defaultValue=0.2, parameterName=None}
    Color=Constant3Vector() {constant=<rgba 0 1 0 0>}
    "f(x)"=Divide(A=@n9) {constB=256}
   ) {materialFunction=&PlotFunctionOnGraph}."f(x) = y"
  )
  n10 = ComponentMask(None=@n3) {r=true}
  Add(A=@n10, B=@n6)
  ComponentMask(None=TextureCoordinate()) {g=true}
  Constant()
  Divide(B=@n10)
  MaterialFunctionCall(Number=Divide(A=ScalarParameter() {defaultValue=3, parameterName=None}) {constB=255}) {materialFunction=&DebugScalarValues}
  Power(Base=OneMinus(None=@n7), Exp=ScalarParameter() {defaultValue=1, parameterName=Power})
  PrecomputedAOMask()
  ScalarParameter() {defaultValue=0.0117, parameterName=None}
  TextureCoordinate()
 ]
}
