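# Unlit material that displays a 3D jump flood result stored in a tiled 2D texture as a signed distance field.
# It samples the texture at texel centers and decodes UVs through EncodeVolumeCoordinates into voxel-center positions.
# The distance to each stored RGB position gets its sign from rounded alpha.
# Further nodes remap that distance by bias-scale, saturation, gradient radius or offset and length, with negative emissive allowed.
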Material /Volumetrics/Tools/JumpFlooding/Materials/M_JumpFlood_Display_3D
graph {
 properties {bAllowNegativeEmissiveColor=true, pixelDepthOffsetMode=PDOM_Legacy, shadingModel=MSM_Unlit, usedShadingModels=Unlit}
 nodes [
  n1 = ComponentMask(None=VectorParameter() {parameterName=TextureSize, defaultValue=<rgba 2048 1024 0 0>}.RGB) {g=true, r=true}
  n2 = Floor(None=Multiply(A=TextureCoordinate(), B=@n1))
  n3 = Divide(B=@n1)
  n4 = TextureSample(UVs=Add(A=Divide(A=@n2, B=@n1), B=Divide(A=@n3))) {samplerType=SAMPLERTYPE_LinearColor, texture=/Volumetrics/Content/VolumeTextures/Textures/RT_Default}
  Ceil(None=@n4.A)
  n5 = VectorParameter() {parameterName=XYFrames, defaultValue=<rgba 16 8 0 0>}
  n6 = VectorParameter() {parameterName=VolumeSize, defaultValue=<rgba 256 256 256 0>}
  n7 = Multiply(
   A=MaterialFunctionCall!(UVs, XY Frames=ComponentMask(None=@n5.RGB) {g=true, r=true}) {materialFunction=/Volumetrics/Content/VolumeTextures/MaterialFunctions/EncodeVolumeCoordinates}.Result
   B=@n6.RGB
  )
  n8 = Add(A=Divide(A=Floor(None=@n7), B=@n6.RGB), B=Divide(B=@n6.RGB) {constA=0.5})
  n9 = Multiply(
   A=MaterialFunctionCall(Vector 3=Subtract(A=@n8, B=@n4.RGB)) {materialFunction=/Engine/Functions/Engine_MaterialFunctions02/Utility/VectorLength}.V3 Length
   B=LinearInterpolate(Alpha=Round(None=@n4.A)) {constA=1, constB=-1}
  )
  Divide(A=@n9, B=ScalarParameter() {defaultValue=0.25, parameterName=Gradient Radius}) {desc=-1 to 1}
  n10 = ConstantBiasScale(None=@n9)
  Divide(A=Subtract(A=@n10, B=ScalarParameter() {defaultValue=0.5, parameterName=Offset}), B=ScalarParameter() {defaultValue=0.5, parameterName=Len})
  MaterialFunctionCall(XY Frames=ComponentMask(None=VectorParameter() {parameterName=XYFrames, defaultValue=<rgba 16 8 0 0>}.RGB) {g=true, r=true}) {materialFunction=&EncodeVolumeCoordinates}
  Saturate(None=@n10)
 ]
}
