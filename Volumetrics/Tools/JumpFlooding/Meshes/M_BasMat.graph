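# M_BasMat is a minimal Default Lit material stored with the meshes of the Volumetrics jump flooding tool, apparently a basic placeholder.
# Its graph holds a single texture sample of the engine's DefaultDiffuse texture, with no connection to any material output shown.
# Usage flags enable it on skeletal meshes and mesh deformers.
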Material /Volumetrics/Tools/JumpFlooding/Meshes/M_BasMat
graph {
 properties {bUsedWithMeshDeformer=true, bUsedWithSkeletalMesh=true, pixelDepthOffsetMode=PDOM_Legacy, usedShadingModels=Default Lit}
 nodes [TextureSample() {texture=/Engine/EngineMaterials/DefaultDiffuse}]
}
