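# Physics asset for the Siren body combined skeletal mesh, stored with the Volumetrics jump-flooding tool meshes.
# It has a single body holding one capsule shape that roughly covers the mesh as a whole.
# There are no constraints, so it gives coarse collision only, with no articulated ragdoll.
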PhysicsAsset /Volumetrics/Tools/JumpFlooding/Meshes/Siren_body_Combined_PhysicsAsset
bodies {Siren_body_Combined [{shapeType=Capsule, length=88.7082, radius=57.1243, center=<0.0048 1.3343 87.3165>, rotation=<rot 2.9061 -90 -0.0839>}]}
constraints []
